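# Sound cue for the shotgun out-of-ammo sound, routed to the third-person ranged weapon fire sound class.
# It randomly picks one of four equally weighted wave variations, then a Modulator varies volume slightly while keeping pitch fixed.
# A concurrency override allows one instance per owner, blocking new plays instead of cutting off the current one, with a short retrigger time.
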SoundCue /Game/Sounds/Fort_Weapons/OutofAmmo/Shotgun_OutofAmmo_Cue
root=Modulator_0
nodes [
 {id=Modulator_0, children [Random_0], properties {pitchMax=1, pitchMin=1, volumeMax=1.1, volumeMin=1}}
 {id=Random_0, children [WavePlayer_1,WavePlayer_2,WavePlayer_3,WavePlayer_4], properties {weights [1,1,1,1]}}
 {id=WavePlayer_1, properties {soundWaveAssetPtr=&/Shotgun_OutofAmmo_01}}
 {id=WavePlayer_2, properties {soundWaveAssetPtr=&/Shotgun_OutofAmmo_02}}
 {id=WavePlayer_3, properties {soundWaveAssetPtr=&/Shotgun_OutofAmmo_03}}
 {id=WavePlayer_4, properties {soundWaveAssetPtr=&/Shotgun_OutofAmmo_04}}
]
properties {
 bOverrideConcurrency=true
 maxDistance=2097152
 soundClassObject=&Fort_WeaponRanged_Fire_3P_PSM
 volumeMultiplier=0.65
 attenuationOverrides {falloffDistance=1600}
 concurrencyOverrides {bLimitToOwner=true, maxCount=1, resolutionRule=PreventNew, retriggerTime=0.3}
}
